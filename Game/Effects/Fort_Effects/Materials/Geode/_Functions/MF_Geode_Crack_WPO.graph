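# Material function for Geode effects that computes a crack world position offset on particle meshes.
# It takes half the vector from particle position to world position, pushing points radially from the particle center, then scales it by Mask, Parameter and CrackWPOMult inputs or parameters.
# A shading path switch zeroes the offset on mobile.
# A directional WPOAmount branch and an actor position node are left unconnected to the output.
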MaterialFunction /Game/Effects/Fort_Effects/Materials/Geode/_Functions/MF_Geode_Crack_WPO
graph {nodes [
  ActorPositionWS()
  n1 = Transform(None=Constant3Vector() {constant=<rgba -1 0 0 0>}) {transformSourceType=TRANSFORMSOURCE_Local, transformType=TRANSFORM_ParticleWorld}
  n2 = WorldPosition()
  n3 = ParticlePositionWS()
  n4 = ShadingPathSwitch(Default=Multiply(A=Subtract(A=@n2.XYZ, B=@n3)) {constB=0.5}, Mobile=Constant()) {desc=Migrated from FeatureLevelSwitch}
  Add(A=Multiply(A=Normalize(VectorInput=@n1), B=ScalarParameter() {defaultValue=400, parameterName=WPOAmount}), B=@n4)
  n5 = FunctionInput(Preview=Constant()) {bUsePreviewValueAsDefault=true, inputName=Parameter, inputType=FunctionInput_Scalar, sortPriority=1}
  n6 = Multiply(A=FunctionInput(Preview=Constant() {r=1}) {bUsePreviewValueAsDefault=true, inputName=Mask, inputType=FunctionInput_Scalar}, B=@n4)
  FunctionOutput(None=Multiply(A=Multiply(A=@n5, B=@n6), B=ScalarParameter() {defaultValue=1, parameterName=CrackWPOMult}))
 ]}
