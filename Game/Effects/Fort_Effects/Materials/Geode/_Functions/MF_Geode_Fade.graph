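# Material function for the Geode rift that collapses and fades materials as they enter or exit the tear, driven by MPC_Geode.
# Clamped rift-relative axis distances drive an XY vertex offset toward the rift, plus opacity outputs for meshes, particle sprites, ribbons and heat shields.
# A smoothstepped band near the rift plane blends animated emissive and normals onto the input material.
# These use panned, world-space triplanar turbulence and glass cracks tinted by MF_Geode_Crack_Colors.
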MaterialFunction /Game/Effects/Fort_Effects/Materials/Geode/_Functions/MF_Geode_Fade
graph {nodes [
  n1 = Constant()
  n2 = ComponentMask(None=CollectionParameter() {collection=/Game/Effects/Fort_Effects/Materials/Geode/MPC_Geode, parameterName=World Location of Rift}) {b=true, g=true, r=true}
  n3 = WorldPosition()
  n4 = AppendVector(A=ComponentMask(None=Subtract(A=@n2, B=@n3.XYZ)) {g=true, r=true}, B=Constant())
  n5 = Subtract(
   A=ComponentMask(None=CollectionParameter() {collection=&MPC_Geode, parameterName=World Location of Rift}) {b=true, g=true, r=true}
   B=WorldPosition().XYZ
  )
  n6 = MaterialFunctionCall(Float3=Transform(None=@n5) {transformSourceType=TRANSFORMSOURCE_World, transformType=TRANSFORM_Local}) {materialFunction=&BreakOutFloat3Components}
  n7 = Clamp(None=@n6.B)
  n8 = CollectionParameter() {collection=&MPC_Geode, parameterName=Entering Or Exiting Tear}
  n9 = LinearInterpolate(A=LinearInterpolate(A=@n4, Alpha=@n7) {constB=0}, B=LinearInterpolate(B=@n4, Alpha=@n7), Alpha=@n8)
  n10 = CollectionParameter() {collection=&MPC_Geode, parameterName=AllowVertCollapseEffect}
  n11 = MaterialFunctionCall!(
   Material=FunctionInput() {inputType=FunctionInput_MaterialAttributes}
   WorldPositionOffset=FunctionOutput(None=LinearInterpolate(A=@n1, B=@n9, Alpha=@n10)) {outputName=WPO, sortPriority=1}
  ) {materialFunction=/Engine/Functions/MaterialLayerFunctions/MatLayerBlend_AddWorldPositionOffset}
  n12 = ScalarParameter() {defaultValue=4.1446, parameterName=RiftColor_Intensity_1}
  n13 = VectorParameter() {parameterName=TextureSizeA, defaultValue=<rgba 1024 1024 1024 0>}
  n14 = Divide(A=WorldPosition().XYZ, B=Multiply(A=Abs(None=@n13.RGB)) {constB=-1})
  n15 = Multiply(A=Time(), B=ScalarParameter() {defaultValue=0.15, parameterName=PanSpeed})
  n16 = Constant2Vector() {g=-10, r=0.1}
  n17 = Panner(Coordinate=ComponentMask(None=@n14) {b=true, r=true}, Time=@n15, Speed=@n16)
  n18 = Panner(Coordinate=ComponentMask(None=@n14) {b=true, g=true}, Time=@n15, Speed=@n16)
  n19 = VertexNormalWS()
  n20 = MaterialFunctionCall(In=Abs(None=ComponentMask(None=@n19) {r=true})) {materialFunction=&CheapContrast}
  n21 = ComponentMask(None=@n14) {g=true, r=true}
  n22 = Abs(None=ComponentMask(None=@n19) {b=true})
  n23 = LinearInterpolate(
   A=LinearInterpolate(
    A=TextureSample(UVs=@n17) {samplerSource=SSM_Wrap_WorldGroupSettings, texture=/Game/Global/Textures/Default/Caustics/T_Turbulence_Seamless}.RGB
    B=TextureSample(UVs=@n18) {samplerSource=SSM_Wrap_WorldGroupSettings, texture=&T_Turbulence_Seamless}.RGB
    Alpha=ComponentMask(None=@n20.Result) {r=true}
   ) {desc=Float3}
   B=TextureSample(UVs=Panner(Coordinate=@n21, Time=@n15, Speed=@n16)) {samplerSource=SSM_Wrap_WorldGroupSettings, texture=&T_Turbulence_Seamless}.RGB
   Alpha=ComponentMask(None=MaterialFunctionCall(In=@n22) {materialFunction=&CheapContrast}.Result) {r=true}
  )
  n24 = Rotator(Coordinate=@n23) {speed=0.5}
  n25 = Multiply(A=@n23) {constB=0.6}
  n26 = TextureSample(UVs=ComponentMask(None=Add(A=@n24, B=@n25)) {g=true, r=true}) {samplerSource=SSM_Wrap_WorldGroupSettings, texture=/Game/Global/Textures/Default/Masks_and_Gradients/T_Grad_Rainbow}
  n27 = MaterialFunctionCall!(Use Alternate UVs, Alternate UVs) {materialFunction=/Game/Athena/Effects/Materials/Geode/Functions/MF_Geode_Crack_Colors}
  n28 = Multiply(A=@n23) {constB=2}
  n29 = Multiply(A=Add(A=Multiply(A=@n26.RGB) {constB=0.2}, B=LinearInterpolate(A=@n27.Cloud Color, B=@n27.Cloud Color, Alpha=@n28)), B=@n28)
  n30 = ScalarParameter() {defaultValue=3, parameterName=CrackColor_Intensity}
  n31 = VectorParameter() {parameterName=TextureSize, defaultValue=<rgba 1024 1024 1024 0>}
  n32 = Divide(A=WorldPosition().XYZ, B=Multiply(A=Abs(None=@n31.RGB)) {constB=-1})
  n33 = Multiply(A=Time(), B=ScalarParameter() {defaultValue=0.15, parameterName=PanSpeed})
  n34 = Constant2Vector() {g=-10, r=0.1}
  n35 = TextureSample(UVs=Panner(Coordinate=ComponentMask(None=@n32) {b=true, g=true}, Time=@n33, Speed=@n34)) {samplerSource=SSM_Wrap_WorldGroupSettings, texture=/Game/Effects/Fort_Effects/Textures/Tile/T_Glass_Cracks_M}
  n36 = TextureSample(UVs=Panner(Coordinate=ComponentMask(None=@n32) {b=true, r=true}, Time=@n33, Speed=@n34)) {samplerSource=SSM_Wrap_WorldGroupSettings, texture=&T_Glass_Cracks_M}
  n37 = VertexNormalWS()
  n38 = ComponentMask(None=@n37) {r=true}
  n39 = ComponentMask(None=MaterialFunctionCall(In=Abs(None=@n38)) {materialFunction=&CheapContrast}.Result) {r=true}
  n40 = LinearInterpolate(A=@n35.G, B=@n36.G, Alpha=@n39)
  n41 = TextureSample(UVs=Panner(Coordinate=ComponentMask(None=@n32) {g=true, r=true}, Time=@n33, Speed=@n34)) {samplerSource=SSM_Wrap_WorldGroupSettings, texture=&T_Glass_Cracks_M}
  n42 = ComponentMask(None=@n37) {b=true}
  n43 = ComponentMask(None=MaterialFunctionCall(In=Abs(None=@n42)) {materialFunction=&CheapContrast}.Result) {r=true}
  n44 = MaterialFunctionCall() {materialFunction=&MF_Geode_Crack_Colors}
  n45 = Multiply(A=@n44.Crack Color) {constB=25}
  n46 = Divide(A=WorldPosition().XYZ, B=Multiply(A=Abs(None=@n31.RGB)) {constB=-1})
  n47 = Multiply(A=Time(), B=ScalarParameter() {defaultValue=0.15, parameterName=PanSpeed_3})
  n48 = Constant2Vector() {g=-10, r=0.1}
  n49 = TextureSample(UVs=Panner(Coordinate=ComponentMask(None=@n46) {b=true, r=true}, Time=@n47, Speed=@n48)) {samplerSource=SSM_Wrap_WorldGroupSettings, samplerType=SAMPLERTYPE_Normal, texture=/Game/Effects/Fort_Effects/Textures/Tile/T_Glass_Cracks_N}
  n50 = TextureSample(UVs=Panner(Coordinate=ComponentMask(None=@n46) {b=true, g=true}, Time=@n47, Speed=@n48)) {samplerSource=SSM_Wrap_WorldGroupSettings, samplerType=SAMPLERTYPE_Normal, texture=&T_Glass_Cracks_N}
  n51 = VertexNormalWS()
  n52 = MaterialFunctionCall(In=Abs(None=ComponentMask(None=@n51) {r=true})) {materialFunction=&CheapContrast}
  n53 = TextureSample(UVs=Panner(Coordinate=ComponentMask(None=@n46) {g=true, r=true}, Time=@n47, Speed=@n48)) {samplerSource=SSM_Wrap_WorldGroupSettings, samplerType=SAMPLERTYPE_Normal, texture=&T_Glass_Cracks_N}
  n54 = Abs(None=ComponentMask(None=@n51) {b=true})
  n55 = LinearInterpolate(
   A=LinearInterpolate(A=@n49.RGB, B=@n50.RGB, Alpha=ComponentMask(None=@n52.Result) {r=true}) {desc=Float3}
   B=@n53.RGB
   Alpha=ComponentMask(None=MaterialFunctionCall(In=@n54) {materialFunction=&CheapContrast}.Result) {r=true}
  )
  n56 = Constant3Vector() {constant=<rgba 0 1 0 0>}
  n57 = Clamp(None=Multiply(A=DotProduct(A=@n55, B=@n56)) {constB=5})
  n58 = Multiply(A=Multiply(A=LinearInterpolate(A=@n40, B=@n41.G, Alpha=@n43)) {constB=2}, B=Add(A=@n44.Cloud Color, B=Multiply(A=@n45, B=@n57) {constB=2}))
  n59 = CollectionParameter() {collection=&MPC_Geode, parameterName=World Location of Rift}
  n60 = Transform(None=Subtract(A=ComponentMask(None=@n59) {b=true, g=true, r=true}, B=WorldPosition().XYZ)) {transformSourceType=TRANSFORMSOURCE_World, transformType=TRANSFORM_Local}
  n61 = Divide(
   A=Distance(
    A=MaterialFunctionCall(Float3=@n60) {materialFunction=&BreakOutFloat3Components}.B
    B=ScalarParameter() {defaultValue=0.025, parameterName=Gradient Fade Distance}
   )
   B=ScalarParameter() {defaultValue=750, parameterName=Threshold Distance}
  )
  n62 = Constant()
  n63 = Constant() {r=1}
  n64 = OneMinus(None=Clamp(None=MaterialFunctionCall(Alpha=@n61, Min=@n62, Max=@n63) {materialFunction=&SmoothStep}.Result))
  n65 = LinearInterpolate(B=Add(A=Multiply(A=@n12, B=@n29), B=Multiply(A=@n30, B=@n58)), Alpha=@n64) {constB=0}
  n66 = Rotator(Coordinate=TextureCoordinate()) {speed=1}
  n67 = Constant()
  n68 = Multiply(
   A=LinearInterpolate(A=LinearInterpolate(A=@n36.R, B=@n35.R, Alpha=@n39) {desc=Float3}, B=@n41.R, Alpha=@n43)
   B=Multiply(A=AppendVector(A=@n66, B=@n67)) {constB=0.3}
  ) {constB=0.2}
  FunctionOutput(None=MaterialFunctionCall!(
    Material=MaterialFunctionCall!(Material=@n11.Blended Material, Emissive=@n65) {materialFunction=/Engine/Functions/MaterialLayerFunctions/MatLayerBlend_Emissive}.Blended Material
    Additional Normal=Add(A=@n68, B=@n55)
    Mask=@n64
   ) {materialFunction=/Engine/Functions/MaterialLayerFunctions/MatLayerBlend_BlendAngleCorrectedNormals}.Blended Material)
  n69 = OneMinus(None=@n7)
  FunctionOutput(None=LinearInterpolate(B=LinearInterpolate(A=@n7, B=@n69, Alpha=@n8), Alpha=@n10) {constA=1}) {outputName=Opacity_Z Axis Meshes, sortPriority=2}
  n70 = Clamp(None=@n6.R)
  n71 = OneMinus(None=@n70)
  FunctionOutput(None=LinearInterpolate(B=LinearInterpolate(A=@n71, B=@n70, Alpha=@n8), Alpha=@n10) {constA=1}) {outputName=Opacity_XAxis Particle Sprites, sortPriority=2}
  n72 = MaterialFunctionCall(Float3=Transform(None=@n5) {transformSourceType=TRANSFORMSOURCE_World, transformType=TRANSFORM_ParticleWorld}) {materialFunction=&BreakOutFloat3Components}
  n73 = Clamp(None=@n72.R)
  n74 = OneMinus(None=@n73)
  FunctionOutput(None=LinearInterpolate(B=LinearInterpolate(A=@n74, B=@n73, Alpha=@n8), Alpha=@n10) {constA=1}) {outputName=Opacity_Mesh Particle Space, sortPriority=2}
  FunctionOutput(None=LinearInterpolate(B=LinearInterpolate(A=@n73, B=@n74, Alpha=@n8), Alpha=@n10) {constA=1}) {outputName=Opacity_Mesh Particle Space_Inverse, sortPriority=2}
  n75 = Clamp(None=@n72.B)
  n76 = OneMinus(None=@n75)
  FunctionOutput(None=LinearInterpolate(B=LinearInterpolate(A=@n75, B=@n76, Alpha=@n8), Alpha=@n10) {constA=1}) {outputName=Opacity_HeatShield Loca, sortPriority=2}
  FunctionOutput(None=LinearInterpolate(B=LinearInterpolate(A=@n70, B=@n71, Alpha=@n8), Alpha=@n10) {constA=1}) {outputName="Opacity_XAxis Particle Sprites_INverted (Ribbon)", sortPriority=2}
 ]}
