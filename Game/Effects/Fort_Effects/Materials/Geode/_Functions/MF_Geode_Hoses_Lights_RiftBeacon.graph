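# Material function that adds animated hose light emissive to Geode Rift Beacon materials.
# A tiled, panning vertical gradient with adjustable contrast forms moving bands, tinted by Geode crack cloud color and scaled by HoseAlpha and green vertex color.
# A ripple-distorted plasma scroll drives a lerp between two identical cloud colors, so it has no visible effect.
# An emissive mask texture gates the result, which blends into the incoming material attributes.
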MaterialFunction /Game/Effects/Fort_Effects/Materials/Geode/_Functions/MF_Geode_Hoses_Lights_RiftBeacon
graph {nodes [
  n1 = Multiply(A=TextureCoordinate(), B=AppendVector(A=Constant() {r=1}, B=ScalarParameter() {defaultValue=2, parameterName=VTile}))
  n2 = Multiply(A=Time(), B=ScalarParameter() {defaultValue=-2, group=Lights, parameterName=Hose_Speed})
  n3 = MaterialFunctionCall(
   In=TextureSample(UVs=Panner(Coordinate=@n1, Time=@n2) {speedY=-1}) {samplerSource=SSM_Wrap_WorldGroupSettings, texture=/Game/Effects/Fort_Effects/Textures/Masks/T_Vert_Gradient_Softer}.R
   Contrast=ScalarParameter() {defaultValue=0.1, parameterName=Contrast_Mask}
  ) {materialFunction=&CheapContrast}
  n4 = MaterialFunctionCall!(Use Alternate UVs, Alternate UVs) {materialFunction=/Game/Athena/Effects/Materials/Geode/Functions/MF_Geode_Crack_Colors}
  n5 = TextureCoordinate() {coordinateIndex=1}
  n6 = Multiply(
   A=@n5
   B=ComponentMask(None=VectorParameter() {group=Hose, parameterName=Liquid_UV_Scale, sortPriority=5, defaultValue=<rgba 8 2 0 0>}.RGB) {g=true, r=true}
  )
  n7 = ComponentMask(None=TextureSample(UVs=@n5) {samplerType=SAMPLERTYPE_Normal, texture=/Game/Effects/Fort_Effects/Gameplay/Water/Textures/T_WaterRipples_01}.RGB) {g=true, r=true}
  n8 = Time()
  n9 = ScalarParameter() {defaultValue=2, group=Hose, parameterName=Flow Speed, sortPriority=4}
  n10 = TextureSample(UVs=Panner(Coordinate=Add(A=@n6, B=@n7), Time=Multiply(A=@n8, B=@n9)) {speedY=-0.5}) {texture=/Game/Effects/Fort_Effects/Textures/Tile/T_Plasma_BallCore}
  n11 = ScalarParameter() {defaultValue=1.25, group=Hose, parameterName=Contrast, sortPriority=3}
  n12 = LinearInterpolate(A=@n4.Cloud Color, B=@n4.Cloud Color, Alpha=Clamp(None=Power(Base=@n10.R, Exp=@n11)))
  n13 = Multiply(A=Multiply(A=Multiply(A=@n3.Result, B=@n12), B=ScalarParameter() {group=Hose, parameterName=HoseAlpha}), B=VertexColor().G)
  n14 = TextureSample(Tex=FunctionInput() {bUsePreviewValueAsDefault=true, inputName=Emissive Mask, inputType=FunctionInput_Texture2D, sortPriority=1}) {texture=/Game/Environments/Props/RiftBeacon/Textures/T_RiftBeaconA_Bits_E}
  FunctionOutput(None=MaterialFunctionCall!(Material=FunctionInput() {inputType=FunctionInput_MaterialAttributes}, Emissive=Multiply(A=@n13, B=@n14.RGB)) {materialFunction=/Engine/Functions/MaterialLayerFunctions/MatLayerBlend_Emissive}.Blended Material)
 ]}
