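# MF_SineVibration is a material function in the Geode effects folder that outputs a time-driven scalar vibration signal.
# It feeds a Time input, defaulting to the material's clock, into three sine waves with unrelated periods, one far faster than the others.
# The two slower waves are scaled to different amplitudes, then all three are summed into an irregular wobble with a fine, rapid jitter.
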MaterialFunction /Game/Effects/Fort_Effects/Materials/Geode/_Functions/MF_SineVibration
graph {nodes [
  n1 = FunctionInput(Preview=Time()) {bUsePreviewValueAsDefault=true, inputName=Time, inputType=FunctionInput_Scalar}
  n2 = Multiply(A=Sine(None=@n1) {period=0.87}) {constB=0.8}
  n3 = Multiply(A=Sine(None=@n1) {period=0.53}) {constB=1.2}
  FunctionOutput(None=Add(A=Add(A=@n2, B=@n3), B=Sine(None=@n1) {period=0.023}))
 ]}
